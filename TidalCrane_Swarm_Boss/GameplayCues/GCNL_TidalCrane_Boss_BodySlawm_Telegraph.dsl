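; Looping gameplay cue that shows the telegraph effect for the TidalCrane Swarm boss's body slam.
; On loop start it spawns the body slam Niagara telegraph at the cue actor's location, keeping only the actor's yaw.
; It gets the current target from MyTarget via an interface message and stretches the mesh along X by the distance to that target, divided by 100.
; The length is set once at spawn and is not updated afterwards.
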(event EventOnLoopingStart(EitherVFX-preferthisversion)
  (params (MyTarget :type Actor%20Object%20Reference) (Parameters :type Gameplay%20Cue%20Parameters%20Structure%20%28by%20ref%29) (ParticleComponents :type Array%20of%20FXSystem%20Component%20Object%20References) (AudioComponents :type Array%20of%20Audio%20Component%20Object%20References))
  (bind _self self)
  (bind _returnvalue (Transformation|GetActorLocation _self))
  (bind _currenttargetactor (Default|GetCurrentTargetActor(Message) MyTarget))
  (bind _returnvalue_2 (Niagara|SpawnSystematLocation 0 "/TidalCrane_Swarm_Boss/Assets/FX/BodySlam/NS_Telegraph_BodySlam01.NS_Telegraph_BodySlam01" _returnvalue (Math|Rotator|MakeRotator 0.0 0.0 (.yaw (Transformation|GetActorRotation _self)))))
  (Niagara|SetNiagaraVariable(Float) _returnvalue_2 "Stretching Mesh by X" (/ (Math|Vector|VectorLength (- (Transformation|GetActorLocation _currenttargetactor) _returnvalue)) 1e+02)))

(fn ConstructionScript)
