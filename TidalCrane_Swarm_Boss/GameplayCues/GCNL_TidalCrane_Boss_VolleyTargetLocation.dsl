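; Looping gameplay cue notify for the Tidal Crane swarm boss that telegraphs where its projectile volley is targeted.
; When looping starts, it gets the volley spawn locations from the target actor through an interface message and spawns a queen telegraph Niagara system at each.
; The spawned components are kept in an array and destroyed when the cue actor ends play, rather than on a cue removal event.
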(event EventOnLoopingStart(EitherVFX-preferthisversion)
  (params (MyTarget :type Actor%20Object%20Reference) (Parameters :type Gameplay%20Cue%20Parameters%20Structure%20%28by%20ref%29) (ParticleComponents :type Array%20of%20FXSystem%20Component%20Object%20References) (AudioComponents :type Array%20of%20Audio%20Component%20Object%20References))
  (bind _projectilevolleyspawnlocations (Default|GetProjectileVolleySpawnLocations(Message) MyTarget))
  (for _array_element _projectilevolleyspawnlocations
    (bind _returnvalue (Niagara|SpawnSystematLocation 0 "/TidalCrane_Swarm_Boss/Assets/FX/NS_Projectile_Queen_Telegraph.NS_Projectile_Queen_Telegraph" _array_element))
    (Utilities|Array|Add (Variables|Default|GetSpawnedVFX) _returnvalue)))

(event EventEndPlay
  (params (EndPlayReason :type EEndPlayReason%20Enum :default "Destroyed"))
  (for _array_element (Variables|Default|GetSpawnedVFX)
    (Components|DestroyComponent _array_element 0)))

(fn ConstructionScript)
