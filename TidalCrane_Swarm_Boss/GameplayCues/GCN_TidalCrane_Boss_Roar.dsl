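; Gameplay cue notify for the TidalCrane swarm boss roar that puts a visual sound indicator on the local player's screen.
; On burst, it calls ShowSoundIndicatorLocal at the target actor's location with a long audible range.
; The indicator uses the Gunshot type with a Queen icon override and references the boss's underground roar sound.
; This burst graph does no other work.
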(fn OnBurst(EitherVFX-preferthisversion)
  (params (MyTarget :type Actor%20Object%20Reference) (Parameters :type Gameplay%20Cue%20Parameters%20Structure%20%28by%20ref%29) (ParticleComponents :type Array%20of%20FXSystem%20Component%20Object%20References) (AudioComponents :type Array%20of%20Audio%20Component%20Object%20References) (BurstCameraShakeInstance :type Camera%20Shake%20Base%20Object%20Reference) (BurstDecalInstance :type Decal%20Actor%20Object%20Reference))
  (Fortnite|ShowSoundIndicatorLocal :WorldContextObject self :Location (Transformation|GetActorLocation MyTarget) :MaxAudibleDistance 5e+04 :IndicatorType "Gunshot" :Tint "(R=1.000000,G=1.000000,B=1.000000,A=1.000000)" :OverrideIcon "/ProtoSwarm/Assets/UI/Icons/T_UI_AV_Queen.T_UI_AV_Queen" :SoundBase "/TidalCrane_Swarm_Boss/Assets/Sound/MSS_TidalCrane_Boss_UndergroundRoar.MSS_TidalCrane_Boss_UndergroundRoar" :AttenuationMaxRangeMultiplier 1.0 :AttenuationIntensityMultiplier 1.0)
  (return))
