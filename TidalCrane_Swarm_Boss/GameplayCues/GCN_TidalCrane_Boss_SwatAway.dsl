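; Burst gameplay cue for the Tidal Crane boss's SwatAway that positions and orients its VFX on the target actor.
; On burst, it takes the first spawned particle component, moves it slightly above the target's location and yaws it a quarter turn from the target's rotation.
; The cue's gameplay effect level doubles as a side selector: level 1 turns the effect negative in yaw, any other level positive.
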(fn OnBurst(EitherVFX-preferthisversion)
  (params (MyTarget :type Actor%20Object%20Reference) (Parameters :type Gameplay%20Cue%20Parameters%20Structure%20%28by%20ref%29) (ParticleComponents :type Array%20of%20FXSystem%20Component%20Object%20References) (AudioComponents :type Array%20of%20Audio%20Component%20Object%20References) (BurstCameraShakeInstance :type Camera%20Shake%20Base%20Object%20Reference) (BurstDecalInstance :type Decal%20Actor%20Object%20Reference))
  (bind (_normalizedmagnitude _rawmagnitude _effectcontext _matchedtagname _originaltag _aggregatedsourcetags _aggregatedtargettags _location _normal _instigator _effectcauser _sourceobject _physicalmaterial _gameplayeffectlevel) (Ability|GameplayCue|BreakGameplayCueParameters Parameters))
  (bind (_returnvalue_roll _returnvalue_pitch _returnvalue_yaw) (Math|Rotator|BreakRotator (Transformation|GetActorRotation MyTarget)))
  (bind _output (Utilities|Array|Get(acopy) ParticleComponents))
  (Transformation|SetWorldRotation _output (Math|Rotator|MakeRotator _returnvalue_roll _returnvalue_pitch (+ _returnvalue_yaw (select (== _gameplayeffectlevel 1) -9e+01 9e+01))) false false)
  (Transformation|SetWorldLocation _output (+ (Transformation|GetActorLocation MyTarget) "0.000000,0.000000,50.000000") false false)
  (return))
